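# Physics asset for the Robot Trouble backpack cosmetic, giving its four small wheels and their two mounts simulated motion.
# A box covers the main dyn_backpack body, and capsules cover the front and back wheel bodies and each wheel.
# Each wheel body hangs off the backpack with locked swing and a narrow twist limit.
# Left and right wheels attach to their wheel body with one free swing axis, letting them spin like wheels on an axle.
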PhysicsAsset /Game/Accessories/FORT_Backpacks/Backpack_M_MED_Robot_Trouble/Meshes/Backpack_M_MED_Robot_Trouble_Physics
bodies {
 dyn_backpack [{shapeType=Box, extentX=23.302, extentY=10.5829, extentZ=34.3092, center=<0 -4.6781 -3.9164>, rotation=<rot 0 0 3.4098>}]
 wheel_b_l [{shapeType=Capsule, length=1.5934, radius=2.3965, center=<0.0088 0.0596 0.0067>, rotation=<rot -79.6014 -12.0861 102.5176>}]
 wheel_b_r [{shapeType=Capsule, length=1.5934, radius=2.4088, center=<0.0004 -0.0607 -0.0106>, rotation=<rot 87.7207 157.3921 78.9519>}]
 wheel_body_b [{shapeType=Capsule, length=5.6709, radius=1.7536, center=<-0.9128 0.0441 -0.705>, rotation=<rot 0 -0.7236 -89.7892>}]
 wheel_body_f [{shapeType=Capsule, length=5.6709, radius=1.8677, center=<-0.7445 -0.0291 0.9955>, rotation=<rot 0 -0.2631 -89.8513>}]
 wheel_f_l [{shapeType=Capsule, length=1.5934, radius=2.3956, center=<-0.0069 0.0591 0.0125>, rotation=<rot -79.7675 172.6839 97.3738>}]
 wheel_f_r [{shapeType=Capsule, length=1.5934, radius=2.3984, center=<0.005 -0.0603 0.0023>, rotation=<rot 87.8827 -3.336 98.1564>}]
}
constraints [
 {
  bone1Name=wheel_body_f
  bone2Name=dyn_backpack
  swing1LimitDegrees=45
  swing1Motion=Locked
  swing2LimitDegrees=45
  swing2Motion=Locked
  twistLimitDegrees=5
 }
 {
  bone1Name=wheel_f_l
  bone2Name=wheel_body_f
  swing1LimitDegrees=45
  swing1Motion=Locked
  swing2LimitDegrees=45
  swing2Motion=Free
  twistLimitDegrees=45
  twistMotion=Locked
 }
 {
  bone1Name=wheel_f_r
  bone2Name=wheel_body_f
  swing1LimitDegrees=45
  swing1Motion=Locked
  swing2LimitDegrees=45
  swing2Motion=Free
  twistLimitDegrees=45
  twistMotion=Locked
 }
 {
  bone1Name=wheel_body_b
  bone2Name=dyn_backpack
  swing1LimitDegrees=45
  swing1Motion=Locked
  swing2LimitDegrees=45
  swing2Motion=Locked
  twistLimitDegrees=5
 }
 {
  bone1Name=wheel_b_l
  bone2Name=wheel_body_b
  swing1LimitDegrees=45
  swing1Motion=Locked
  swing2LimitDegrees=45
  swing2Motion=Free
  twistLimitDegrees=45
  twistMotion=Locked
 }
 {
  bone1Name=wheel_b_r
  bone2Name=wheel_body_b
  swing1LimitDegrees=45
  swing1Motion=Locked
  swing2LimitDegrees=45
  swing2Motion=Free
  twistLimitDegrees=45
  twistMotion=Locked
 }
]
